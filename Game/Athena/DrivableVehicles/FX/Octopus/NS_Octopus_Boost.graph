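# Niagara boost effect for the Octopus (Baller) drivable vehicle: blue glow, light, trail and speed lines.
# Baller_Glow mixes an initial mesh burst with continuous spawns, styled separately through spawn-group masks.
# Light flickers, its radius shrinking with emitter age, while RibbonTrail spawns by distance and passes User.BoostRemaining to its material.
# Speedlines_Camera bursts camera-offset sprites on a cylinder, pushed backward and stretched by speed.
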NiagaraSystem /Game/Athena/DrivableVehicles/FX/Octopus/NS_Octopus_Boost
dependencies {
 usedDataInterfaces [
  &NiagaraDataInterfaceCurve
  &NiagaraDataInterfaceVector2DCurve
  &NiagaraDataInterfaceVectorCurve
 ]
 usedDynamicInputs []
 usedModules [&SystemState]
}
emitters {
 Baller_Glow {
  emitterUpdateScript [
   &EmitterState()
   &SpawnRate(SpawnRate=10, Use Spawn Probability=false)
   &SpawnBurst_Instantaneous(Spawn Group=1, Use Spawn Probability=false, UseLoopCountLimit=true)
  ]
  particleSpawnScript [
   &InitializeParticle(
    Color=&MaskLinearColorBySpawnGroup(Masked Linear Color=<rgba 0.5314 1.9654 5 0.25>, Passthrough Linear Color=<rgba 2.1258 7.8616 20 1>)
    Lifetime Max=&MaskFloatBySpawnGroup(Masked Float=0.5, Passthrough Float=0.2)
    Lifetime Min:NiagaraFloat=&MaskFloatBySpawnGroup(Masked Float=0.35, Passthrough Float=0.15)
    Lifetime Mode=::NewEnumerator1
    Mesh Scale Mode=::NewEnumerator3
    Mesh Uniform Scale=1.75
    UsePositionOffset=false
    Write Material Random=true
   )
   &InitialMeshOrientation(
    Mesh Orientation Mode=::NewEnumerator1
    Rotation=&MaskVectorBySpawnGroup(Masked Vector=&RandomRangeVector(Minimum=<0 0 0>, Maximum=<1 1 1>), Passthrough Vector=<0 0.25 0>)
    Use Rotation Vector=true
   )
  ]
  particleUpdateScript [
   &ParticleState()
   &ScaleColor(
    Scale Alpha=&MaskFloatBySpawnGroup(
     Masked Float=&FloatFromCurve(FloatCurve=DataInterface() {propertyValues {Curve {keys [
          {interpMode=RCIM_Cubic, tangentMode=RCTM_User, time=0, value=0}
          {interpMode=RCIM_Cubic, tangentMode=RCTM_User, time=0.5, value=1}
          {tangentMode=RCTM_User, time=1, value=0}
         ]}}})
     Passthrough Float=1
    )
    Scale RGB=&VectorFromCurve(VectorCurve=DataInterface() {propertyValues {
       XCurve {keys [<at 0 1 RCIM_Cubic>,<at 1 0.0643 RCIM_Cubic>]}
       YCurve {keys [<at 0 1 RCIM_Cubic>,<at 1 1 RCIM_Cubic>]}
       ZCurve {keys [<at 0 1 RCIM_Cubic>,<at 1 0.2771 RCIM_Cubic>]}
      }})
    ScaleA=true
    ScaleRGB=true
   )
   &DynamicMaterialParameters(
    Index 0 Param 1=1
    Index 0 Param 2=&FloatFromCurve(FloatCurve=DataInterface() {propertyValues {Curve {keys [<at 0 1>,<at 1 -1>]}}})
    Index 0 Param 3=1
    Index 0 Param 4=20
    Module.Param0WriteEnabled=true
    Module.Param1WriteEnabled=true
    Module.Param2WriteEnabled=true
    Module.Param3WriteEnabled=true
   )
   &ScaleMeshSize(Scale Factor=&VectorFromFloat(Value=&FloatFromCurve(FloatCurve=DataInterface() {propertyValues {Curve {keys [<at 0 0.9>,<at 1 1.07>]}}})))
  ]
  rendererClasses [&NiagaraMeshRendererProperties]
  renderers [{rendererClass=&NiagaraMeshRendererProperties, rendererIndex=0}]
 }
 Light {
  emitterUpdateScript [
   &EmitterState()
   &SpawnRate(SpawnRate=5, Use Spawn Probability=false)
  ]
  particleSpawnScript [
   &InitializeParticle(Color=<rgba 3.9921 14.1525 32 1>, Lifetime=0.22, UsePositionOffset=false, Write Material Random=true)
   /SetVariables(Particles.LightRadius=&FloatFromCurve(FloatCurve=DataInterface() {propertyValues {Curve {keys [<at 0.1 230 RCIM_Cubic>,<at 1.7 120>]}}}, CurveIndex=$Emitter.LoopedAge)) {isSetParameters=true}
  ]
  particleUpdateScript [
   &ParticleState()
   &ScaleColor(
    Scale RGB=&VectorFromFloat(Value=&FloatFromCurve(FloatCurve=DataInterface() {propertyValues {Curve {keys [
          <at 0 0.5 RCIM_Cubic>
          {arriveTangent=1.9776, interpMode=RCIM_Cubic, leaveTangent=1.9776, time=0.0785, value=0.7603}
          {arriveTangent=-0.268, interpMode=RCIM_Cubic, leaveTangent=-0.268, time=0.2003, value=0.8962}
          {arriveTangent=0.2307, interpMode=RCIM_Cubic, leaveTangent=0.2307, time=0.2972, value=0.7017}
          {arriveTangent=0.6306, interpMode=RCIM_Cubic, leaveTangent=0.6306, time=0.3988, value=0.942}
          {arriveTangent=0.788, interpMode=RCIM_Cubic, leaveTangent=0.788, time=0.4731, value=0.8127}
          {arriveTangent=1.3168, interpMode=RCIM_Cubic, leaveTangent=1.3168, time=0.5352, value=1.0495}
          {arriveTangent=0.2706, interpMode=RCIM_Cubic, leaveTangent=0.2706, time=0.5991, value=0.9786}
          {arriveTangent=-0.4427, interpMode=RCIM_Cubic, leaveTangent=-0.4427, time=0.6578, value=1.0827}
          {arriveTangent=-0.9788, interpMode=RCIM_Cubic, leaveTangent=-0.9788, time=0.7273, value=0.9219}
          {arriveTangent=-0.318, interpMode=RCIM_Cubic, leaveTangent=-0.318, time=0.7845, value=0.9587}
          {arriveTangent=-2.1284, interpMode=RCIM_Cubic, leaveTangent=-2.1284, time=0.9051, value=0.8653}
          <at 1 0.5 RCIM_Cubic>
         ]}}}))
    ScaleA=true
    ScaleRGB=true
   )
  ]
  rendererClasses [&NiagaraLightRendererProperties]
  renderers [{rendererClass=&NiagaraLightRendererProperties, rendererIndex=0}]
 }
 RibbonTrail {
  emitterUpdateScript [
   &EmitterState()
   &SpawnPerUnit(Module.Use Max Movement Threshold=true, Module.Use Movement Tolerance=false, Use Spawn Probability=false)
  ]
  particleSpawnScript [
   &InitializeRibbon(Color=<rgba 0.0995 0.5523 1 0.25>, Lifetime=0.25, Ribbon Width=100, UsePositionOffset=false, Write Material Random=true)
   &AddVelocity(Velocity=<-1600 0 0>)
  ]
  particleUpdateScript [
   &ParticleState()
   &Drag(Drag=7, Use Linear Drag=true, Use Rotational Drag=false)
   &SolveForcesAndVelocity(Clamp Velocity=false, Limit Acceleration=false)
   &ScaleColor(
    Scale Alpha=&FloatFromCurve(FloatCurve=DataInterface() {propertyValues {Curve {keys [
         {arriveTangent=3.554, interpMode=RCIM_Cubic, leaveTangent=3.554, tangentMode=RCTM_User, time=0, value=0}
         {
          arriveTangent=2.5474
          arriveTangentWeight=0.1292
          interpMode=RCIM_Cubic
          leaveTangent=2.5474
          leaveTangentWeight=0.0447
          tangentMode=RCTM_User
          time=0.2438
          value=0.9559
         }
         {arriveTangent=-2.196, interpMode=RCIM_Cubic, leaveTangent=-2.196, tangentMode=RCTM_User, time=0.3282, value=0.4056}
         <at 1 0 RCIM_Cubic>
        ]}}})
    ScaleA=true
    ScaleRGB=true
   )
   &DynamicMaterialParameters(
    Index 0 Param 1=$User.BoostRemaining
    Index 0 Param 2=1
    Index 0 Param 3=1
    Index 0 Param 4=1
    Module.Param0WriteEnabled=true
    Module.Param1WriteEnabled=true
    Module.Param2WriteEnabled=true
    Module.Param3WriteEnabled=true
   )
  ]
  rendererClasses [&NiagaraRibbonRendererProperties]
  renderers [{rendererClass=&NiagaraRibbonRendererProperties}]
 }
 Speedlines_Camera {
  emitterUpdateScript [
   &EmitterState()
   &SpawnBurst_Instantaneous(Spawn Count=20, Spawn Time=0.015, Use Spawn Probability=false, UseLoopCountLimit=true)
   &SpawnBurst_Instantaneous(Spawn Count=20, Spawn Time=0.1, Use Spawn Probability=false, UseLoopCountLimit=true) {module=SpawnBurst_Instantaneous001}
  ]
  particleSpawnScript [
   &InitializeParticle(
    Color=&MakeLinearColorFromVectorAndFloat(
     "Vector (RGB)"=&MakeVectorFromLinearColorRGB(Color=<rgba 0.6875 0.6851 0.6839 1>)
     "Float (Alpha)"=&RandomRangeFloat(Minimum=0.25, Maximum=0.4)
    )
    Lifetime Max=0.35
    Lifetime Min=0.25
    Lifetime Mode=::NewEnumerator1
    Sprite Size Max=<7.5 7>
    Sprite Size Min=<6 4>
    Sprite Size Mode=::NewEnumerator2
    UsePositionOffset=false
    Write Material Random=true
   )
   &CameraOffset(Camera Offset Amount=-100)
   &ShapeLocation(
    Cylinder Radius=&RandomRangeFloat(Minimum=150, Maximum=200)
    MaskSpawnGroup=false
    Override Seed=false
    Rotation Mode=::NewEnumerator1
    Shape Primitive=::NewEnumerator1
    Surface Only=true
    Use Endcaps In Surface Only Mode=false
    UseOwnerScale=true
   )
   &AddVelocity(Velocity=&RandomRangeVector(Minimum=<-1100 0 0>, Maximum=<-1600 0 0>))
   &AddVelocity(Velocity=&Multiply_VectorByFloat(Vector=$Particles.ShapeLocation.ShapeNormal, Float=200)) {module=AddVelocity001}
  ]
  particleUpdateScript [
   &ParticleState()
   &InheritVelocity(Inherited Velocity Amount Scale=<-0.7 -0.7 -0.7>, Speed Limit=true, UseSpeedThreshold=false)
   &SolveForcesAndVelocity(Clamp Velocity=false, Limit Acceleration=false)
   &ScaleColor(
    Scale Alpha=&FloatFromCurve(FloatCurve=DataInterface() {propertyValues {Curve {keys [
         {interpMode=RCIM_Cubic, tangentMode=RCTM_User, time=0, value=0}
         {interpMode=RCIM_Cubic, tangentMode=RCTM_User, time=0.4449, value=0.904}
         {tangentMode=RCTM_User, time=1, value=0}
        ]}}})
    ScaleA=true
    ScaleRGB=true
   )
   &ScaleSpriteSizeBySpeed(Max Scale Factor=<1 100>)
   &ScaleSpriteSize(
    Non-Uniform Curve Sprite Scale=DataInterface() {propertyValues {
      XCurve {keys [<at 0 0.5 RCIM_Cubic>,<at 0.5 1 RCIM_Cubic>,<at 1 0.5 RCIM_Cubic>]}
      YCurve {keys [<at 0 0 RCIM_Cubic>,<at 0.5 1 RCIM_Cubic>,<at 1 0 RCIM_Cubic>]}
     }}
    Scale Sprite Size Mode=::NewEnumerator3
   )
  ]
  rendererClasses [&NiagaraSpriteRendererProperties]
  renderers [{rendererClass=&NiagaraSpriteRendererProperties}]
 }
}
summary {
 systemName=NS_Octopus_Boost
 emitters [
  {emitterName=RibbonTrail, rendererClasses [&NiagaraRibbonRendererProperties]}
  {emitterName=Speedlines_Camera, rendererClasses [&NiagaraSpriteRendererProperties]}
  {emitterName=Baller_Glow, rendererClasses [&NiagaraMeshRendererProperties]}
  {emitterName=Light, rendererClasses [&NiagaraLightRendererProperties]}
 ]
 userVariables [User.BoostRemaining {type=NiagaraFloat, defaultValue=1, description=""}]
}
