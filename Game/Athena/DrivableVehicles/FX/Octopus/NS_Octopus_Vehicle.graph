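# Niagara system for the Octopus drivable vehicle's effects, including glowing coils, electric arcs, damage sparks and smoke, and skid dust and debris.
# Coil, arc and damage emitters follow sockets through a shared skeletal mesh reader, choosing them by spawn group or particle partition.
# User parameters drive the effects: Boost scales arcs, Damage and BadlyDamaged scale sparks and smoke, SpawnDust and Dust control skid effects, and Fast feeds coil materials.
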NiagaraSystem /Game/Athena/DrivableVehicles/FX/Octopus/NS_Octopus_Vehicle
dependencies {
 usedDataInterfaces [
  &NiagaraDataInterfaceArrayFloat3
  &NiagaraDataInterfaceCurve
  &NiagaraDataInterfaceParticleRead
  &NiagaraDataInterfaceArrayFloat
  &NiagaraDataInterfaceColorCurve
  &NiagaraDataInterfaceVectorCurve
  &NiagaraDataInterfaceVector2DCurve
  &NiagaraDataInterfaceSkeletalMesh
 ]
 usedDynamicInputs [&MakeVector4FromVectorAndFloat]
 usedModules [
  &SystemState
  &self:SystemSpawnScript.SystemScriptSource.SystemScriptGraph.NiagaraNodeAssignment_0.SetVariables_#1
  &self:SystemSpawnScript.SystemScriptSource.SystemScriptGraph.NiagaraNodeAssignment_1.SetVariables_#2
  &self:SystemSpawnScript.SystemScriptSource.SystemScriptGraph.NiagaraNodeAssignment_2.SetVariables_#3
  &self:SystemSpawnScript.SystemScriptSource.SystemScriptGraph.NiagaraNodeAssignment_3.SetVariables_#4
  &self:SystemSpawnScript.SystemScriptSource.SystemScriptGraph.NiagaraNodeAssignment_4.SetVariables_#5
  &self:SystemSpawnScript.SystemScriptSource.SystemScriptGraph.NiagaraNodeAssignment_5.SetVariables_#6
 ]
}
emitters {
 Coil_Meshes {
  emitterUpdateScript [
   &EmitterState(
    Inactive Response=::NewEnumerator1
    Life Cycle Mode=::NewEnumerator1
    Loop Behavior=::NewEnumerator1
    Loop Duration Mode=::NewEnumerator1
    UseLoopDelay=false
   )
   &SpawnBurst_Instantaneous(Spawn Count=4, Use Spawn Probability=false, UseLoopCountLimit=false) {module=SpawnBurst_Instantaneous008}
   &SpawnBurst_Instantaneous(Use Spawn Probability=false, UseLoopCountLimit=false) {enabled=false}
   &SpawnBurst_Instantaneous(Spawn Group=1, Use Spawn Probability=false, UseLoopCountLimit=false) {enabled=false, module=SpawnBurst_Instantaneous001}
   &SpawnBurst_Instantaneous(Spawn Group=2, Use Spawn Probability=false, UseLoopCountLimit=false) {enabled=false, module=SpawnBurst_Instantaneous002}
   &SpawnBurst_Instantaneous(Spawn Group=3, Use Spawn Probability=false, UseLoopCountLimit=false) {enabled=false, module=SpawnBurst_Instantaneous003}
   &SpawnBurst_Instantaneous(Spawn Group=4, Use Spawn Probability=false, UseLoopCountLimit=false) {enabled=false, module=SpawnBurst_Instantaneous004}
   &SpawnBurst_Instantaneous(Spawn Group=5, Use Spawn Probability=false, UseLoopCountLimit=false) {enabled=false, module=SpawnBurst_Instantaneous005}
   &SpawnBurst_Instantaneous(Spawn Group=6, Use Spawn Probability=false, UseLoopCountLimit=false) {enabled=false, module=SpawnBurst_Instantaneous006}
   &SpawnBurst_Instantaneous(Spawn Group=7, Use Spawn Probability=false, UseLoopCountLimit=false) {enabled=false, module=SpawnBurst_Instantaneous007}
  ]
  particleSpawnScript [
   /SetVariables(Particles.VisibilityTag=&MakeCustomIntFromBool(
     Bool=&SetBoolByIntComparison(A=&SpawnGroup(), B=4, Comparison Type=::NewEnumerator2)
     True INT=1
     False INT=0
    )) {isSetParameters=true}
   &InitializeParticle(
    Color=<rgba 2 2 2 10>
    Lifetime=0
    Mesh Scale Mode=::NewEnumerator3
    Mesh Uniform Scale=0.85
    Position Mode=::NewEnumerator0
    Sprite Size Mode=::NewEnumerator3
    Uniform Sprite Size=30
    Write Material Random=true
   )
   &CameraOffset(Camera Offset Amount=3)
   &FN_SetSpawnGroup()
  ]
  particleUpdateScript [
   &ParticleState(Kill Particles When Lifetime Has Elapsed=false, Let Infinitely Lived Particles Die When Emitter Deactivates=true)
   &UpdateMeshOrientation(Rotation Rate=&RandomRangeFloat(Minimum=-0.5, Maximum=0.5))
   &AlignSpriteToMeshOrientation()
   &ScaleColor(Scale Alpha=$System.SharedAlphaControl, ScaleA=true, ScaleRGB=true)
   /SetVariables(
    Particles.Flare Camera Offset=2
    Particles.Flare Color:LinearColor=&MakeLinearColorFromVectorAndFloat(
     "Vector (RGB)"=&MakeVectorFromLinearColorRGB(Color=<rgba 2.7604 8.1095 10 1>)
     "Float (Alpha)"=&Multiply_Float(A=1, B=$System.SharedAlphaControl)
    )
    Particles.Flare Size:Vector2f=&Vector2DFromFloat(Value=&RandomRangeFloat(Minimum=40, Maximum=50))
    Particles.Sprite Camera Offset=18
    Particles.SpriteColor=&MakeLinearColorFromVectorAndFloat(
     "Vector (RGB)"=&MakeVectorFromLinearColorRGB(Color=<rgba 0.3642 0.6292 1 1>)
     "Float (Alpha)"=&Multiply_Float(A=2, B=$System.SharedAlphaControl)
    )
   ) {isSetParameters=true}
   &DynamicMaterialParameters(
    Index 0 Param 1=3
    Index 0 Param 2=1
    Index 0 Param 3=1
    Index 0 Param 4=1
    Index 1 Param 1=&RandomRangeFloat(Minimum=0, Maximum=1)
    Index 1 Param 2=&RemapRange(Input Value To Remap=$User.Fast, Input Min=0, Input Max=0.2, Output Min=0, Output Max=1)
    Index 2 Param 1=0.5
    Index 2 Param 2=0.125
    Index 2 Param 3=1
    Index 2 Param 4=1
    Module.Param0WriteEnabled=true
    Module.Param0WriteEnabled1=true
    Module.Param0WriteEnabled2=true
    Module.Param1WriteEnabled=true
    Module.Param1WriteEnabled1=true
    Module.Param1WriteEnabled2=true
    Module.Param2WriteEnabled=true
    Module.Param2WriteEnabled1=true
    Module.Param2WriteEnabled2=true
    Module.Param3WriteEnabled=true
    Module.Param3WriteEnabled1=true
    Module.Param3WriteEnabled2=true
    Write Parameter Index 1=true
    Write Parameter Index 2=true
   )
   &SkeletalMeshLocation(
    Bone / Socket Index=&MakeCustomIntFromBool(
     Bool=&SetBoolByIntComparison(A=&SpawnGroup(), B=4, Comparison Type=::NewEnumerator2)
     True INT=/Niagara/DynamicInputs/Subtract/Subtract_Integer(A:NiagaraInt32=&SpawnGroup(), B:NiagaraInt32=4)
     False INT=&SpawnGroup()
    )
    MaskSpawnGroup=false
    Mesh Sampling Type=::NewEnumerator5
    Sampled Position Offset=&SelectVectorFromArray(Direct Array Index=$Particles.VisibilityTag, Vector Selection Array=DataInterface() {propertyValues {FloatData [<-4.4253 0 0>, <0 0 0>]}})
    Skeletal Mesh=$System.Skeletal Mesh Reader
    Socket Sampling Mode=::NewEnumerator5
   )
   &FN_Sprite_Attributes(
    Create Position Attribute=true
    Module Output:ENiagara_AttributeSamplingApplyOutput=::NewEnumerator1
    Position:NiagaraPosition=&AddVectorToPosition(
     Position=$Particles.Position
     Vector=&RotateVectorByQuaternion(Vector To Rotate=<4 0 0>, Quaternion=$Particles.SkeletalMeshLocation.SampledRotation)
    )
   )
  ]
  rendererClasses [&NiagaraMeshRendererProperties, &NiagaraSpriteRendererProperties]
  renderers [
   {rendererClass=&NiagaraMeshRendererProperties, rendererIndex=0}
   {rendererClass=&NiagaraSpriteRendererProperties, rendererIndex=1}
   {rendererClass=&NiagaraSpriteRendererProperties, rendererIndex=2}
  ]
 }
 DamagePoint {
  simTarget=GPUComputeSim
  emitterUpdateScript [
   &EmitterState(Life Cycle Mode=::NewEnumerator1, Loop Duration=1, UseLoopDelay=false)
   &SpawnBurst_Instantaneous(
    Spawn Count=&MakeIntFromFloat(Float=&Multiply_Float(A=4, B=$User.Damage))
    Use Spawn Probability=false
    UseLoopCountLimit=false
   )
   &SpawnBurst_Instantaneous(
    Spawn Count=&MakeIntFromFloat(Float=&Multiply_Float(A=16, B=$User.Damage))
    Spawn Group=1
    Use Spawn Probability=false
    UseLoopCountLimit=false
   ) {module=SpawnBurst_Instantaneous001}
   &SpawnRate(SpawnRate=&Multiply_Float(A=1, B=$User.Damage), Use Spawn Probability=false) {enabled=false}
   &SpawnRate(SpawnGroup=1, SpawnRate=&Multiply_Float(A=1, B=$User.Damage), Use Spawn Probability=false) {enabled=false, module=SpawnRate001}
   &SpawnRate(SpawnGroup=2, SpawnRate=&Multiply_Float(A=1, B=$User.Damage), Use Spawn Probability=false) {enabled=false, module=SpawnRate002}
   &SpawnRate(SpawnGroup=3, SpawnRate=&Multiply_Float(A=1, B=$User.Damage), Use Spawn Probability=false) {enabled=false, module=SpawnRate003}
   &SpawnRate(SpawnRate=&Multiply_Float(A=1, B=$User.Damage), Use Spawn Probability=false) {enabled=false, module=SpawnRate004}
  ]
  particleSpawnScript [
   &InitializeParticle(
    Color=&MaskLinearColor(Mask Source=0, Mask Parameter=?, Mask=0, Masked Linear Color=<rgba 0.1131 0.4052 1 1>, Passthrough Linear Color=<rgba 1.4583 4.8588 8 1>)
    Lifetime=&MaskFloat(
     Mask Source=0
     Mask Parameter=?
     Mask=0
     Masked Float=0.1
     Passthrough Float=&RandomRangeFloat(Minimum=0.075, Maximum=0.15)
    )
    Sprite Size Mode=::NewEnumerator1
    Uniform Sprite Size Max=70
    Uniform Sprite Size Min=50
    UsePositionOffset=false
    Write Material Random=true
   )
   &PartitionParticles(Number Of Partitions=4)
   &SkeletalMeshLocation(
    Bone / Socket Index=$Particles.PartitionParticles.Partition
    MaskSpawnGroup=false
    Mesh Sampling Type=::NewEnumerator5
    Position Sampling=::NewEnumerator1
    Skeletal Mesh=$System.Skeletal Mesh Reader
    Socket Sampling Mode=::NewEnumerator5
   ) {module=SkeletalMeshLocation001}
   &CameraOffset(Camera Offset Amount=10)
   &FN_SetSpawnGroup() {enabled=false}
   /SetVariables(Particles.VisibilityTag=&SpawnGroup()) {isSetParameters=true}
   /SetVariables(Particles.SpriteSize=&MaskVector2dBySpawnGroup(
     Masked Vector2D=&Vector2DFromFloat(Value=&RandomRangeFloat(Minimum=50, Maximum=70))
     Passthrough Vector2D=&RandomRangeVector2D(Minimum=<0.5 1>, Maximum=<1 3>)
     Spawn Group Mask=0
    )) {isSetParameters=true}
   &ShapeLocation(
    MaskSpawnGroup=false
    Override Seed=false
    Sphere Radius=&MaskFloatBySpawnGroup(Masked Float=0, Passthrough Float=&RandomRangeFloat(Minimum=5, Maximum=15))
    UseOwnerScale=true
   ) {module=ShapeLocation001}
   &AddVelocity(
    Module.Use Falloff Distance=false
    Override Seed=false
    Velocity Mode=::NewEnumerator1
    Velocity Origin=$Particles.SkeletalMeshLocation001.SampledPosition
    Velocity Speed=&MaskFloat(
     Mask Source=0
     Mask Parameter=?
     Mask=0
     Masked Float=0
     Passthrough Float=&RandomRangeFloat(Minimum=50, Maximum=150)
    )
   )
   &AddVelocity(Velocity=&MaskVector(
     Mask Source=0
     Mask Parameter=?
     Mask=0
     Masked Vector=<0 0 0>
     Passthrough Vector=&RandomRangeVector(Minimum=<0 -50 0>, Maximum=<200 -150 0>)
    )) {module=AddVelocity001}
   &ApplyInitialForces() {enabled=false}
  ]
  particleUpdateScript [
   &ParticleState()
   &SkeletalMeshLocation(
    Bone / Socket Index=$Particles.PartitionParticles.Partition
    MaskSpawnGroup=false
    Mesh Sampling Type=::NewEnumerator5
    Position Sampling=::NewEnumerator1
    Sampled Position Offset=&MaskVectorBySpawnGroup(
     Masked Vector=&RandomRangeVector(Minimum=<-5 -5 -5>, Maximum=<5 5 5>)
     Passthrough Vector=&RandomRangeVector(Minimum=<-10 -10 -10>, Maximum=<10 10 10>)
    )
    Skeletal Mesh=$System.Skeletal Mesh Reader
    Socket Sampling Mode=::NewEnumerator5
   )
   &ShapeLocation(
    Cylinder Height=&MaskFloat(Mask Source=0, Mask Parameter=?, Mask=0, Masked Float=5, Passthrough Float=0)
    Cylinder Radius=&MaskFloat(Mask Source=0, Mask Parameter=?, Mask=0, Masked Float=15, Passthrough Float=0)
    MaskSpawnGroup=false
    Override Seed=false
    Rotation Coordinate Space=::Simulation
    Rotation Mode=::NewEnumerator1
    Shape Origin=$Particles.SkeletalMeshLocation.SampledPosition
    Shape Primitive=::NewEnumerator1
    Surface Only=false
    UseOwnerScale=true
   ) {enabled=false}
   &ScaleColor(
    Scale Alpha=&FloatFromCurve(FloatCurve=DataInterface() {propertyValues {Curve {keys [
         {arriveTangent=9.0725, interpMode=RCIM_Cubic, leaveTangent=9.0725, tangentMode=RCTM_User, time=0, value=0}
         {arriveTangent=-2.5249, interpMode=RCIM_Cubic, leaveTangent=-2.5249, tangentMode=RCTM_User, time=0.05, value=1}
         {arriveTangent=-0.0405, interpMode=RCIM_Cubic, leaveTangent=-0.0405, tangentMode=RCTM_User, time=1, value=0}
        ]}}})
    Scale RGB=&VectorFromFloat(Value=&FloatFromCurve(FloatCurve=DataInterface() {propertyValues {Curve {keys [
          {arriveTangent=9.0725, interpMode=RCIM_Cubic, leaveTangent=9.0725, tangentMode=RCTM_User, time=0, value=1}
          {arriveTangent=-2.5249, interpMode=RCIM_Cubic, leaveTangent=-2.5249, tangentMode=RCTM_User, time=0.1, value=6}
          {arriveTangent=-0.0405, interpMode=RCIM_Cubic, leaveTangent=-0.0405, tangentMode=RCTM_User, time=1, value=1}
         ]}}}))
    ScaleA=true
    ScaleRGB=true
   )
   &AccelerationForce(Acceleration=&MaskVector(Mask Source=0, Mask Parameter=?, Mask=0, Masked Vector=<0 0 0>, Passthrough Vector=<0 0 -1600>))
   &SolveForcesAndVelocity(Clamp Velocity=false, Limit Acceleration=false)
   &ScaleSpriteSizeBySpeed(
    Max Scale Factor=&MaskVector2dBySpawnGroup(Masked Vector2D=<1 1>, Passthrough Vector2D=<1 10>, Spawn Group Mask=0)
    Sample Scale Factor By Curve=true
    Scale Factor Curve=DataInterface() {propertyValues {Curve {keys [{arriveTangent=2.405, interpMode=RCIM_Cubic, leaveTangent=2.405, tangentMode=RCTM_User, time=0, value=0}, <at 1 1 RCIM_Cubic>]}}}
   )
  ]
  rendererClasses [&NiagaraSpriteRendererProperties]
  renderers [
   {rendererClass=&NiagaraSpriteRendererProperties}
   {rendererClass=&NiagaraSpriteRendererProperties, rendererIndex=1}
  ]
 }
 DamageSparks {
  bEnabled=false
  simTarget=GPUComputeSim
  emitterUpdateScript [
   &EmitterState()
   &SpawnParticlesFromOtherEmitter(
    Attribute Reader=DataInterface() {propertyValues {EmitterBinding {bindingMode=Other, emitterName=DamagePoint}}}
    Spawn Rate=50
    Use Maximum Age=false
    Use Minimum Age=false
   )
  ]
  particleSpawnScript [
   &InitializeParticle(
    Color=<rgba 1.4583 4.8588 8 1>
    Lifetime Max=0.15
    Lifetime Min=0.075
    Lifetime Mode=::NewEnumerator1
    Sprite Size Max=<1 3>
    Sprite Size Min=<0.5 1>
    Sprite Size Mode=::NewEnumerator2
    UsePositionOffset=false
    Write Material Random=true
   )
   &SampleParticlesFromOtherEmitter()
   &ShapeLocation(
    MaskSpawnGroup=false
    Override Seed=false
    Shape Origin=$Particles.Position
    Sphere Radius=&RandomRangeFloat(Minimum=5, Maximum=15)
    UseOwnerScale=true
   )
   &AddVelocity(Velocity=&Multiply_VectorByFloat(Vector=$Particles.ShapeLocation.ShapeNormal, Float=&RandomRangeFloat(Minimum=25, Maximum=80)))
   &AddVelocity(Velocity=&RandomRangeVector(Minimum=<-200 -200 0>, Maximum=<200 200 200>)) {module=AddVelocity001}
  ]
  particleUpdateScript [
   &ParticleState()
   &AccelerationForce(Acceleration=<0 0 -1600>)
   &SolveForcesAndVelocity(Clamp Velocity=false, Limit Acceleration=false)
   &ScaleSpriteSize(Uniform Curve Sprite Scale=DataInterface() {propertyValues {Curve {keys [
        {arriveTangent=96.9815, interpMode=RCIM_Cubic, leaveTangent=96.9815, tangentMode=RCTM_User, time=0, value=0}
        {arriveTangentWeight=0.0272, interpMode=RCIM_Cubic, leaveTangentWeight=0.8781, tangentMode=RCTM_User, time=0.03, value=1}
        {arriveTangent=-0.0405, interpMode=RCIM_Cubic, leaveTangent=-0.0405, tangentMode=RCTM_User, time=1, value=0}
       ]}}})
   &ScaleSpriteSizeBySpeed(
    Max Scale Factor=<1 10>
    Sample Scale Factor By Curve=true
    Scale Factor Curve=DataInterface() {propertyValues {Curve {keys [{arriveTangent=2.405, interpMode=RCIM_Cubic, leaveTangent=2.405, tangentMode=RCTM_User, time=0, value=0}, <at 1 1 RCIM_Cubic>]}}}
   )
  ]
  rendererClasses [&NiagaraSpriteRendererProperties]
  renderers [{rendererClass=&NiagaraSpriteRendererProperties}]
 }
 Damaged_Smoke {
  emitterUpdateScript [
   &EmitterState(Life Cycle Mode=::NewEnumerator1, Loop Duration=1, UseLoopDelay=false)
   &SpawnRate(SpawnRate=&Multiply_Float(A=20, B=$User.BadlyDamaged), Use Spawn Probability=false)
   &SpawnRate(SpawnRate=&Multiply_Float(A=5, B=$User.BadlyDamaged), Use Spawn Probability=false) {enabled=false, module=SpawnRate008}
   &SpawnRate(SpawnGroup=1, SpawnRate=&Multiply_Float(A=5, B=$User.BadlyDamaged), Use Spawn Probability=false) {enabled=false, module=SpawnRate001}
   &SpawnRate(SpawnGroup=2, SpawnRate=&Multiply_Float(A=5, B=$User.BadlyDamaged), Use Spawn Probability=false) {enabled=false, module=SpawnRate002}
   &SpawnRate(SpawnGroup=3, SpawnRate=&Multiply_Float(A=5, B=$User.BadlyDamaged), Use Spawn Probability=false) {enabled=false, module=SpawnRate003}
   &SpawnRate(SpawnGroup=5, SpawnRate=&Multiply_Float(A=5, B=$User.BadlyDamaged), Use Spawn Probability=false) {enabled=false, module=SpawnRate005}
   &SpawnRate(SpawnGroup=6, SpawnRate=&Multiply_Float(A=5, B=$User.BadlyDamaged), Use Spawn Probability=false) {enabled=false, module=SpawnRate006}
   &SpawnRate(SpawnGroup=7, SpawnRate=&Multiply_Float(A=5, B=$User.BadlyDamaged), Use Spawn Probability=false) {enabled=false, module=SpawnRate007}
   &SpawnRate(SpawnGroup=4, SpawnRate=&Multiply_Float(A=5, B=$User.BadlyDamaged), Use Spawn Probability=false) {enabled=false, module=SpawnRate004}
  ]
  particleSpawnScript [
   /SetVariables(Particles.VisibilityTag=&MakeCustomIntFromBool(
     Bool=&SetBoolByIntComparison(A=&SpawnGroup(), B=4, Comparison Type=::NewEnumerator2)
     True INT=1
     False INT=0
    )) {isSetParameters=true}
   &InitializeParticle(
    Color=<rgba 0.0156 0.0156 0.0156 1.25>
    Lifetime Max=&SelectFloatFromArray(Direct Array Index=$Particles.VisibilityTag, Float Selection Array=DataInterface() {propertyValues {FloatData [0.9,0.5]}})
    Lifetime Min:NiagaraFloat=&SelectFloatFromArray(Direct Array Index=$Particles.VisibilityTag, Float Selection Array=DataInterface() {propertyValues {FloatData [0.45,0.15]}})
    Lifetime Mode=::NewEnumerator1
    Sprite Rotation Mode=::NewEnumerator2
    Sprite Size Mode=::NewEnumerator1
    Uniform Sprite Size Max=75
    Uniform Sprite Size Min=30
    UsePositionOffset=false
    Write Material Random=true
   )
   &CameraOffset(Camera Offset Amount=25)
   &AddVelocity(Velocity=&MakeCustomVectorFromBool(
     Bool=&SetBoolByIntComparison(A=&SpawnGroup(), B=4, Comparison Type=::NewEnumerator2)
     True Vector=&RandomRangeVector(Minimum=<-50 0 150>, Maximum=<-100 0 0>)
     False Vector=&RandomRangeVector(Minimum=<0 0 10>, Maximum=<0 0 30>)
    ))
   &FN_SetSpawnGroup()
  ]
  particleUpdateScript [
   &ParticleState()
   &SkeletalMeshLocation(
    Bone / Socket Index=&MakeCustomIntFromBool(
     Bool=&SetBoolByIntComparison(A=&SpawnGroup(), B=4, Comparison Type=::NewEnumerator2)
     True INT=&Subtract_Integer(A:NiagaraInt32=&SpawnGroup(), B:NiagaraInt32=4)
     False INT=&SpawnGroup()
    )
    MaskSpawnGroup=false
    Mesh Sampling Type=::NewEnumerator5
    Position Sampling=::NewEnumerator1
    Skeletal Mesh=$System.Skeletal Mesh Reader
    Socket Sampling Mode=::NewEnumerator5
   )
   &ShapeLocation(
    Cylinder Height=2
    Cylinder Radius=&MakeFloatFromInt(Int=&MakeCustomIntFromBool(
      Bool=&SetBoolByIntComparison(A=&SpawnGroup(), B=4, Comparison Type=::NewEnumerator2)
      True INT=10
      False INT=5
     ))
    MaskSpawnGroup=false
    Override Seed=false
    Rotation Coordinate Space=::Simulation
    Rotation Mode=::NewEnumerator1
    Shape Origin=$Particles.SkeletalMeshLocation.SampledPosition
    Shape Primitive=::NewEnumerator1
    Surface Only=true
    Surface Only Band Thickness=0.25
    Use Endcaps In Surface Only Mode=false
    UseOwnerScale=true
   )
   &AccelerationForce(Acceleration=&SelectVectorFromArray(Direct Array Index=$Particles.VisibilityTag, Vector Selection Array=DataInterface() {propertyValues {FloatData [<0 0 100>, <0 0 1500>]}}))
   &VectorNoiseForce(
    Force Amount=&MakeCustomFloatFromBool(
     Bool=&SetBoolByIntComparison(A=&SpawnGroup(), B=4, Comparison Type=::NewEnumerator2)
     True Float=1500
     False Float=0
    )
    Module.Use Random Noise Time Interval=false
   )
   &Drag(
    Drag=&SelectFloatFromArray(Direct Array Index=$Particles.VisibilityTag, Float Selection Array=DataInterface() {propertyValues {FloatData [2,1.5]}})
    Use Linear Drag=true
    Use Rotational Drag=false
   )
   &SolveForcesAndVelocity(Clamp Velocity=false, Limit Acceleration=false)
   &SubUVAnimation(
    End Frame=55
    Start Frame=10
    SubUV Animation Mode:ENiagara_SubUVLookupMode=::NewEnumerator1
    SubUV Curve=DataInterface() {propertyValues {Curve {keys [<at 0 0 RCIM_Cubic>,<at 1 1 RCIM_Cubic>]}}}
   )
   &SpriteRotationRate(Rotation Rate=&RandomRangeFloat(Minimum=-0.5, Maximum=0.5))
   &ScaleColor(
    Scale Alpha=&FloatFromCurve(FloatCurve=DataInterface() {propertyValues {Curve {keys [
         {interpMode=RCIM_Cubic, leaveTangentWeight=0.0333, time=0, value=0}
         {arriveTangent=1.3067, arriveTangentWeight=0.0548, interpMode=RCIM_Cubic, leaveTangent=1.3067, leaveTangentWeight=0.249, time=0.1, value=1}
         {
          arriveTangent=-1.1111
          arriveTangentWeight=0.2262
          interpMode=RCIM_Cubic
          leaveTangent=-1.1111
          leaveTangentWeight=0.2222
          time=0.554
          value=0.7239
         }
         {arriveTangentWeight=0.1487, interpMode=RCIM_Cubic, time=1, value=0}
        ]}}})
    Scale RGB=&MakeVectorFromLinearColorRGB(Color=&ColorFromCurve(ColorCurve=DataInterface() {propertyValues {
        AlphaCurve {keys []}
        BlueCurve {keys [<at 0 5 RCIM_Cubic>,<at 0.5 1 RCIM_Cubic>]}
        GreenCurve {keys [<at 0 4.9186 RCIM_Cubic>,<at 0.5 1 RCIM_Cubic>]}
        RedCurve {keys [<at 0 4.8698 RCIM_Cubic>,<at 0.5 1 RCIM_Cubic>]}
       }}))
    ScaleA=true
    ScaleRGB=true
   )
   &ScaleSpriteSize(Uniform Curve Sprite Scale=DataInterface() {propertyValues {Curve {keys [{arriveTangent=0.8132, interpMode=RCIM_Cubic, leaveTangent=0.8132, tangentMode=RCTM_User, time=0, value=0.3}, <at 1 1 RCIM_Cubic>]}}})
   &DynamicMaterialParameters(
    Index 0 Param 1=&FloatFromCurve(FloatCurve=DataInterface() {propertyValues {Curve {keys [<at 0 1>,<at 1 -1>]}}})
    Index 0 Param 2=1
    Index 0 Param 3=1
    Index 0 Param 4=1
    Module.Param0WriteEnabled=true
    Module.Param1WriteEnabled=true
    Module.Param2WriteEnabled=true
    Module.Param3WriteEnabled=true
   )
   /SetVariables(
    Particles.SparkColor:LinearColor=&MultiplyLinearColor(
     Linear Color=<rgba 2 0.2858 0.0464 1>
     Scale Factor=&MakeFloat4FromVectorAndFloat(
      "Vector (RGB)"=&VectorFromCurve(VectorCurve=DataInterface() {propertyValues {
         XCurve {keys [<at 0 16 RCIM_Cubic>,<at 0.8 1 RCIM_Cubic>]}
         YCurve {keys [<at 0 16 RCIM_Cubic>,<at 0.8 0.1249 RCIM_Cubic>]}
         ZCurve {keys [<at 0 16 RCIM_Cubic>,<at 0.8 0.2547 RCIM_Cubic>]}
        }})
      "Float (Alpha)"=&FloatFromCurve(FloatCurve=DataInterface() {propertyValues {Curve {keys [
           {arriveTangent=-13.2873, interpMode=RCIM_Cubic, leaveTangent=-13.2873, leaveTangentWeight=0.0667, tangentMode=RCTM_User, time=0, value=2}
           {arriveTangentWeight=0.0667, interpMode=RCIM_Cubic, tangentMode=RCTM_User, time=0.2, value=1}
          ]}}})
     )
    )
    Particles.SparkSize:Vector2f=&self:SparkSpeed(
     Initial Scale:Vector2f=&Multiply_Vector2D_ByFloat(
      Float=1
      Vector2D=&Multiply_Vector2D_ByFloat(
       Float=&FloatFromCurve(FloatCurve=DataInterface() {propertyValues {Curve {keys [
            {arriveTangent=99.8643, interpMode=RCIM_Cubic, leaveTangent=99.8643, tangentMode=RCTM_User, time=0, value=0}
            {arriveTangent=-2.5249, interpMode=RCIM_Cubic, leaveTangent=-2.5249, tangentMode=RCTM_User, time=0.03, value=1}
            {arriveTangent=-0.0405, interpMode=RCIM_Cubic, leaveTangent=-0.0405, tangentMode=RCTM_User, time=1, value=0}
           ]}}})
       Vector2D=&RandomRangeVector2D(Minimum=<1 1>, Maximum=<2 2>)
      )
     )
     Max Scale Factor:Vector2f=<1 5>
     Min Scale Factor:Vector2f=<1 15>
     Velocity Threshold:NiagaraFloat=0
    )
   ) {isSetParameters=true}
  ]
  rendererClasses [&NiagaraSpriteRendererProperties]
  renderers [
   {rendererClass=&NiagaraSpriteRendererProperties}
   {rendererClass=&NiagaraSpriteRendererProperties, rendererIndex=1}
  ]
 }
 Directional_Smoke {
  emitterUpdateScript [
   &EmitterState(Life Cycle Mode=::NewEnumerator1, Loop Delay=0.05, Loop Duration=0.5, UseLoopDelay=true)
   &SpawnRate(
    SpawnRate=&MakeCustomFloatFromBool(
     Bool=$User.SpawnDust
     True Float=&RemapRange(Input Value To Remap=$User.Dust, Input Min=0, Input Max=1, Output Min=0, Output Max=60)
     False Float=0
    )
    Use Spawn Probability=false
   )
  ]
  particleSpawnScript [
   &InitializeParticle(
    Color=&MakeLinearColorFromVectorAndFloat(
     "Vector (RGB)"=&MakeVectorFromLinearColorRGB(Color=&RandomRangeLinearColor(Minimum=<rgba 0.1771 0.1469 0.1328 1>, Maximum=<rgba 0.5 0.4135 0.371 1>))
     "Float (Alpha)"=&RandomRangeFloat(Minimum=0.05, Maximum=0.005)
    )
    Lifetime Max=0.4
    Lifetime Min=0.2
    Lifetime Mode=::NewEnumerator1
    Sprite Size Max=<120 300>
    Sprite Size Min=<80 220>
    Sprite Size Mode=::NewEnumerator2
    UsePositionOffset=false
    Write Material Random=true
   )
   &SystemLocation(MaskSpawnGroup=false, Offset=$System.Location)
   &CameraOffset(Camera Offset Amount=25)
   &AddVelocity(Velocity=&RandomRangeVector(Minimum=<-700 -160 150>, Maximum=<-500 160 450>))
  ]
  particleUpdateScript [
   &Drag(Drag=3, Use Linear Drag=true, Use Rotational Drag=false)
   &InheritVelocity(Inherited Velocity Amount Scale=<0.25 0.25 0>, Speed Limit=true, UseSpeedThreshold=false)
   &SolveForcesAndVelocity(Clamp Velocity=false, Limit Acceleration=false)
   &ParticleState()
   &ScaleColor(
    Scale Alpha=&FloatFromCurve(
     FloatCurve=DataInterface() {propertyValues {Curve {keys [{interpMode=RCIM_Cubic, tangentMode=RCTM_User, time=0, value=1}, <at 1 0>]}}}
     Scale Curve=50
    )
    Scale RGB=<2 2 2>
    ScaleA=true
    ScaleRGB=true
   )
   &ScaleSpriteSize(Scale Factor=&Vector2DFromCurve(
     Vector2Curve=DataInterface() {propertyValues {
       XCurve {keys [
         {arriveTangent=0.5563, interpMode=RCIM_Cubic, leaveTangent=0.5563, leaveTangentWeight=0.1606, tangentMode=RCTM_User, time=0, value=0.5}
         {
          arriveTangent=1.0085
          arriveTangentWeight=0.1796
          interpMode=RCIM_Cubic
          leaveTangent=1.0085
          leaveTangentWeight=0.1931
          tangentMode=RCTM_User
          time=0.4819
          value=0.7227
         }
         {arriveTangentWeight=0.1727, interpMode=RCIM_Cubic, time=1, value=1}
        ]}
       YCurve {keys [
         {arriveTangent=1.0323, interpMode=RCIM_Cubic, leaveTangent=1.0323, tangentMode=RCTM_User, time=0, value=0.3991}
         {arriveTangent=1.4348, interpMode=RCIM_Cubic, leaveTangent=1.4348, tangentMode=RCTM_User, time=0.4819, value=0.7355}
         <at 1 1.4633 RCIM_Cubic>
        ]}
      }}
     ScaleCurve=<1 1>
    ))
   &DynamicMaterialParameters(
    Index 0 Param 1=&Add_Float(
     A=$Particles.RandomFloat
     B=&FloatFromCurve(FloatCurve=DataInterface() {propertyValues {Curve {keys [
          {arriveTangent=2.7533, interpMode=RCIM_Cubic, leaveTangent=2.847, tangentMode=RCTM_Break, time=0, value=0}
          {arriveTangent=1.3581, interpMode=RCIM_Cubic, leaveTangent=1.3581, tangentMode=RCTM_User, time=0.2558, value=0.5125}
          {arriveTangent=0.0529, interpMode=RCIM_Cubic, leaveTangent=0.2021, tangentMode=RCTM_User, time=1, value=1}
         ]}}})
    )
    Index 0 Param 4=&FloatFromCurve(FloatCurve=DataInterface() {propertyValues {Curve {keys [{arriveTangent=-4.5888, interpMode=RCIM_Cubic, leaveTangent=-4.5888, tangentMode=RCTM_User, time=0, value=1}, <at 0.8 0 RCIM_Cubic>]}}})
    Index 2 Param 1=&FloatFromCurve(FloatCurve=DataInterface() {propertyValues {Curve {keys [
         {arriveTangent=1.629, interpMode=RCIM_Cubic, leaveTangent=1.629, tangentMode=RCTM_User, time=0, value=0}
         {interpMode=RCIM_Cubic, leaveTangent=-0.0719, time=0.8, value=1}
        ]}}})
    Index 2 Param 2=&FloatFromCurve(
     FloatCurve=DataInterface() {propertyValues {Curve {keys [
         {arriveTangent=1.9056, time=0, value=0.2}
         {arriveTangent=0.0259, interpMode=RCIM_Cubic, leaveTangent=0.0259, tangentMode=RCTM_User, time=0.8, value=1.4}
        ]}}}
     Scale Curve=&RandomRangeFloat(Minimum=0.5, Maximum=1)
    )
    Index 2 Param 3=&FloatFromCurve(FloatCurve=DataInterface() {propertyValues {Curve {keys [
         {interpMode=RCIM_Cubic, leaveTangent=1.6063, tangentMode=RCTM_Break, time=0, value=1}
         {interpMode=RCIM_Cubic, leaveTangent=11.1233, tangentMode=RCTM_Break, time=0.8, value=2}
        ]}}})
    Index 2 Param 4=&FloatFromCurve(
     FloatCurve=DataInterface() {propertyValues {Curve {keys [
         {arriveTangent=-0.6867, interpMode=RCIM_Cubic, leaveTangent=-0.6867, tangentMode=RCTM_User, time=0, value=1}
         {interpMode=RCIM_Cubic, leaveTangent=-1.2956, tangentMode=RCTM_Break, time=0.8, value=0.75}
        ]}}}
     Scale Curve=1.6
    )
    Module.Param0WriteEnabled=true
    Module.Param0WriteEnabled1=true
    Module.Param0WriteEnabled2=true
    Module.Param1WriteEnabled=false
    Module.Param1WriteEnabled1=true
    Module.Param1WriteEnabled2=true
    Module.Param2WriteEnabled=true
    Module.Param2WriteEnabled1=true
    Module.Param2WriteEnabled2=true
    Module.Param3WriteEnabled=true
    Module.Param3WriteEnabled1=true
    Module.Param3WriteEnabled2=true
    Write Parameter Index 1=true
    Write Parameter Index 2=true
   )
  ]
  rendererClasses [&NiagaraSpriteRendererProperties]
  renderers [
   {rendererClass=&NiagaraSpriteRendererProperties}
   {rendererClass=&NiagaraSpriteRendererProperties, rendererIndex=1}
  ]
 }
 Electric_Arcs {
  emitterUpdateScript [
   &EmitterState()
   &SpawnRate(
    SpawnRate=&Multiply_Float(
     A=40
     B=&RemapRange(Input Value To Remap=$User.Boost, Input Min=0, Input Max=1, Output Min=0.1, Output Max=1)
    )
    Use Spawn Probability=false
   ) {module=SpawnRate004}
   &SpawnRate(
    SpawnRate=&Multiply_Float(
     A=10
     B=&RemapRange(Input Value To Remap=$User.Boost, Input Min=0, Input Max=1, Output Min=0.1, Output Max=1)
    )
    Use Spawn Probability=false
   ) {enabled=false}
   &SpawnRate(
    SpawnGroup=1
    SpawnRate=&Multiply_Float(
     A=10
     B=&RemapRange(Input Value To Remap=$User.Boost, Input Min=0, Input Max=1, Output Min=0.05, Output Max=1)
    )
    Use Spawn Probability=false
   ) {enabled=false, module=SpawnRate001}
   &SpawnRate(
    SpawnGroup=2
    SpawnRate=&Multiply_Float(
     A=10
     B=&RemapRange(Input Value To Remap=$User.Boost, Input Min=0, Input Max=1, Output Min=0.05, Output Max=1)
    )
    Use Spawn Probability=false
   ) {enabled=false, module=SpawnRate002}
   &SpawnRate(
    SpawnGroup=3
    SpawnRate=&Multiply_Float(
     A=10
     B=&RemapRange(Input Value To Remap=$User.Boost, Input Min=0, Input Max=1, Output Min=0.05, Output Max=1)
    )
    Use Spawn Probability=false
   ) {enabled=false, module=SpawnRate003}
  ]
  particleSpawnScript [
   &InitializeParticle(
    Lifetime Max=1
    Lifetime Min=0.2
    Lifetime Mode=::NewEnumerator1
    Mesh Scale Max=<0.15 0.3 0.3>
    Mesh Scale Min=<0.125 0.225 0.225>
    Mesh Scale Mode=::NewEnumerator2
    UsePositionOffset=false
    Write Material Random=true
   )
   &CameraOffset(Camera Offset Amount=3)
   &InitialMeshOrientation(
    Mesh Orientation Mode=::NewEnumerator2
    Rotation=&RandomRangeVector(Minimum=<0 0 0>, Maximum=<1 0 0>)
    Use Rotation Vector=true
   )
   /SetVariables(Particles.LightExponent=2, Particles.LightRadius=&RandomRangeFloat(Minimum=15, Maximum=20)) {isSetParameters=true}
   &FN_SetSpawnGroup()
  ]
  particleUpdateScript [
   &ParticleState()
   &SkeletalMeshLocation(
    Bone / Socket Index=&SpawnGroup()
    MaskSpawnGroup=false
    Mesh Sampling Type=::NewEnumerator5
    Sampled Position Offset=<-5 0 0>
    Skeletal Mesh=$System.Skeletal Mesh Reader
    Socket Sampling Mode=::NewEnumerator5
   )
   &ShapeLocation(
    Apply To Particle Position=::NewEnumerator1
    Cylinder Height=5
    Cylinder Radius=15
    MaskSpawnGroup=false
    Override Seed=false
    Rotation Coordinate Space=::Simulation
    Rotation Mode=::NewEnumerator1
    Scale Mode:ENiagara_ScaleMode=::NewEnumerator1
    Shape Origin=$Particles.SkeletalMeshLocation.SampledPosition
    Shape Primitive=::NewEnumerator1
    Surface Only=true
    Use Endcaps In Surface Only Mode=false
   )
   &DynamicMaterialParameters(
    Index 0 Param 1=&FloatFromCurve(FloatCurve=DataInterface() {propertyValues {Curve {keys [
         {interpMode=RCIM_Cubic, tangentMode=RCTM_User, time=0, value=0}
         {interpMode=RCIM_Cubic, tangentMode=RCTM_User, time=0.5, value=1}
         {tangentMode=RCTM_User, time=1, value=0}
        ]}}})
    Index 0 Param 2=&RandomRangeFloat(Minimum=0, Maximum=1)
    Index 0 Param 3=8
    Index 0 Param 4=1
    Module.Param0WriteEnabled=true
    Module.Param1WriteEnabled=true
    Module.Param2WriteEnabled=true
    Module.Param3WriteEnabled=true
   )
   &ScaleColor(
    Scale Alpha=&FloatFromCurve(FloatCurve=DataInterface() {propertyValues {Curve {keys [
         <at 0 1>
         {arriveTangent=-4.949, interpMode=RCIM_Cubic, leaveTangent=-4.949, time=0.1502, value=1}
         {arriveTangentWeight=0.137, interpMode=RCIM_Cubic, leaveTangentWeight=0.2007, tangentMode=RCTM_User, time=0.2009, value=0.0057}
         {arriveTangentWeight=0.0252, interpMode=RCIM_Cubic, leaveTangentWeight=0.0154, tangentMode=RCTM_User, time=0.2752, value=-0.0053}
         {arriveTangentWeight=0.0152, interpMode=RCIM_Cubic, leaveTangentWeight=0.018, tangentMode=RCTM_User, time=0.3205, value=-0.0174}
         {arriveTangentWeight=0.0193, interpMode=RCIM_Cubic, leaveTangentWeight=0.018, tangentMode=RCTM_User, time=0.374, value=0.0081}
         {arriveTangentWeight=0.017, interpMode=RCIM_Cubic, leaveTangentWeight=0.0118, tangentMode=RCTM_User, time=0.4239, value=0.0248}
         {arriveTangentWeight=0.1439, interpMode=RCIM_Cubic, leaveTangentWeight=0.1818, tangentMode=RCTM_User, time=0.4583, value=-0.0105}
         {arriveTangentWeight=0.0838, interpMode=RCIM_Cubic, leaveTangentWeight=0.1711, tangentMode=RCTM_User, time=0.5018, value=0.9987}
         {arriveTangentWeight=0.0296, interpMode=RCIM_Cubic, leaveTangentWeight=0.046, tangentMode=RCTM_User, time=0.5906, value=0.7426}
         {arriveTangentWeight=0.0951, interpMode=RCIM_Cubic, leaveTangentWeight=0.1875, tangentMode=RCTM_User, time=0.7284, value=0.9897}
         <at 1 0 RCIM_Cubic>
        ]}}})
    ScaleA=true
    ScaleRGB=true
   )
   &ScaleMeshSize(Scale Factor=&MakeVector(
     X=1
     Y=1
     Z=&FloatFromCurve(FloatCurve=DataInterface() {propertyValues {Curve {keys [<at 0 0.9>,<at 1 1.2>]}}})
    ))
   &UpdateMeshOrientation(Rotation Rate=&RandomRangeFloat(Minimum=-0.1, Maximum=0.1))
   /SetVariables(
    Particles.LightColor=&MultiplyLinearColor(
     Linear Color=<rgba 0.11 0.5168 1 1>
     Scale Factor=&MakeFloat4FromVectorAndFloat(
      "Vector (RGB)"=&VectorFromFloat(Value=&FloatFromCurve(
        FloatCurve=DataInterface() {propertyValues {Curve {keys [
            <at 0 1>
            {arriveTangent=-4.949, interpMode=RCIM_Cubic, leaveTangent=-4.949, time=0.1502, value=1}
            {arriveTangentWeight=0.137, interpMode=RCIM_Cubic, leaveTangentWeight=0.2007, tangentMode=RCTM_User, time=0.2009, value=0.0057}
            {arriveTangentWeight=0.0252, interpMode=RCIM_Cubic, leaveTangentWeight=0.0154, tangentMode=RCTM_User, time=0.2752, value=-0.0053}
            {arriveTangentWeight=0.0152, interpMode=RCIM_Cubic, leaveTangentWeight=0.018, tangentMode=RCTM_User, time=0.3205, value=-0.0174}
            {arriveTangentWeight=0.0193, interpMode=RCIM_Cubic, leaveTangentWeight=0.018, tangentMode=RCTM_User, time=0.374, value=0.0081}
            {arriveTangentWeight=0.017, interpMode=RCIM_Cubic, leaveTangentWeight=0.0118, tangentMode=RCTM_User, time=0.4239, value=0.0248}
            {arriveTangentWeight=0.1439, interpMode=RCIM_Cubic, leaveTangentWeight=0.1818, tangentMode=RCTM_User, time=0.4583, value=-0.0105}
            {arriveTangentWeight=0.0838, interpMode=RCIM_Cubic, leaveTangentWeight=0.1711, tangentMode=RCTM_User, time=0.5018, value=0.9987}
            {arriveTangentWeight=0.0296, interpMode=RCIM_Cubic, leaveTangentWeight=0.046, tangentMode=RCTM_User, time=0.5906, value=0.7426}
            {arriveTangentWeight=0.0951, interpMode=RCIM_Cubic, leaveTangentWeight=0.1875, tangentMode=RCTM_User, time=0.7284, value=0.9897}
            <at 1 0 RCIM_Cubic>
           ]}}}
        Scale Curve=30
       ))
      "Float (Alpha)"=1
     )
    )
    Particles.LightPosition:NiagaraPosition=$Particles.ShapeLocation.ShapePosition
   ) {isSetParameters=true}
  ]
  rendererClasses [&NiagaraMeshRendererProperties, &NiagaraLightRendererProperties]
  renderers [
   {rendererClass=&NiagaraMeshRendererProperties, rendererIndex=0}
   {rendererClass=&NiagaraLightRendererProperties, rendererIndex=1}
  ]
 }
 Empty {bEnabled=false, rendererClasses [], renderers []}
 Empty001 {bEnabled=false, rendererClasses [], renderers []}
 Skid_Chunks {
  simTarget=GPUComputeSim
  emitterUpdateScript [
   &EmitterState()
   &SpawnRate(SpawnRate=&Multiply_Float(A=50, B=$User.Dust), Use Spawn Probability=false) {enabled=false}
   &SpawnRate(
    SpawnRate=&MakeCustomFloatFromBool(Bool=$User.SpawnDust, True Float=&Multiply_Float(A=50, B=$User.Dust), False Float=0)
    Use Spawn Probability=false
   ) {module=SpawnRate001}
   &SpawnRate(
    SpawnGroup=1
    SpawnRate=&MakeCustomFloatFromBool(Bool=$User.SpawnDust, True Float=&Multiply_Float(A=10, B=$User.Dust), False Float=0)
    Use Spawn Probability=false
   ) {module=SpawnRate002}
  ]
  particleSpawnScript [
   /SetVariables(Particles.VisibilityTag=&SpawnGroup()) {isSetParameters=true}
   &InitializeParticle(
    Color=&MaskLinearColorBySpawnGroup(Masked Linear Color=<rgba 0.0312 0.0312 0.0312 1>, Passthrough Linear Color=<rgba 0.1042 0.1042 0.1042 2>)
    Lifetime=&MaskFloatBySpawnGroup(
     Masked Float=&RandomRangeFloat(Minimum=0.2, Maximum=0.8)
     Passthrough Float=&RandomRangeFloat(Minimum=0.4, Maximum=0.7)
    )
    Sprite Rotation Mode=::NewEnumerator2
    Sprite Size Max=&MaskVector2dBySpawnGroup(Masked Vector2D=<9 9>, Passthrough Vector2D=<6 15>, Spawn Group Mask=0)
    Sprite Size Min=&MaskVector2dBySpawnGroup(Masked Vector2D=<2 2>, Passthrough Vector2D=<4 5>, Spawn Group Mask=0)
    Sprite Size Mode=::NewEnumerator2
    UsePositionOffset=false
    Write Material Random=true
   )
   &SystemLocation(MaskSpawnGroup=false, Offset=$System.Location)
   &AddVelocity(Velocity=&MaskVectorBySpawnGroup(
     Masked Vector=&RandomRangeVector(Minimum=<-500 -90 60>, Maximum=<-50 90 700>)
     Passthrough Vector=&RandomRangeVector(Minimum=<-150 -150 250>, Maximum=<0 150 700>)
    ))
  ]
  particleUpdateScript [
   &AccelerationForce(Acceleration=&MaskVectorBySpawnGroup(Masked Vector=<0 0 -1200>, Passthrough Vector=<0 0 -1600>))
   &SolveForcesAndVelocity(Clamp Velocity=false, Limit Acceleration=false)
   &ParticleState()
   &SubUVAnimation(End Frame=3)
   &SpriteRotationRate(Rotation Rate=&MaskFloatBySpawnGroup(
     Masked Float=&RandomRangeFloat(Minimum=0.5, Maximum=3)
     Passthrough Float=&RandomRangeFloat(Minimum=-6, Maximum=6)
    ))
   &ScaleSpriteSize(
    Scale Sprite Size Mode=::NewEnumerator0
    Uniform Scale Factor=&MaskFloatBySpawnGroup(
     Masked Float=1
     Passthrough Float=&FloatFromCurve(FloatCurve=DataInterface() {propertyValues {Curve {keys [<at 0.7 1 RCIM_Cubic>,<at 1 0>]}}})
    )
   )
   &DynamicMaterialParameters(
    Index 0 Param 1=&RandomRangeFloat(Minimum=0, Maximum=1)
    Index 0 Param 2=1
    Index 0 Param 3=1
    Index 0 Param 4=1
    Module.Param0WriteEnabled=true
    Module.Param1WriteEnabled=true
    Module.Param2WriteEnabled=true
    Module.Param3WriteEnabled=true
   )
  ]
  rendererClasses [&NiagaraSpriteRendererProperties]
  renderers [
   {rendererClass=&NiagaraSpriteRendererProperties}
   {rendererClass=&NiagaraSpriteRendererProperties, rendererIndex=1}
  ]
 }
 Skid_Mesh {
  emitterUpdateScript [
   &EmitterState(
    Life Cycle Mode=::NewEnumerator1
    Loop Duration=&RandomRangeFloat(Minimum=0.5, Maximum=1.2)
    Recalculate Duration Each Loop=true
    UseLoopDelay=false
   )
   &SpawnRate(
    SpawnRate=&MakeCustomFloatFromBool(
     Bool=$User.SpawnDust
     True Float=&RemapRange(Input Value To Remap=$User.Dust, Input Min=0, Input Max=1, Output Min=0, Output Max=30)
     False Float=0
    )
    Use Spawn Probability=false
   )
   &SpawnRate(
    SpawnGroup=1
    SpawnRate=&MakeCustomFloatFromBool(
     Bool=$User.SpawnDust
     True Float=&RemapRange(Input Value To Remap=$User.Dust, Input Min=0, Input Max=1, Output Min=0, Output Max=10)
     False Float=0
    )
    Use Spawn Probability=false
   ) {module=SpawnRate001}
  ]
  particleSpawnScript [
   &InitializeParticle(
    Color=&MakeLinearColorFromVectorAndFloat(
     "Vector (RGB)"=&MakeVectorFromLinearColorRGB(Color=&MaskLinearColorBySpawnGroup(Masked Linear Color=<rgba 0.1094 0.0772 0.064 1>, Passthrough Linear Color=<rgba 0.2292 0.1788 0.1671 1>))
     "Float (Alpha)"=&MaskFloatBySpawnGroup(
      Masked Float=&RandomRangeFloat(Minimum=1.2, Maximum=1.4)
      Passthrough Float=&RandomRangeFloat(Minimum=0.5, Maximum=0.7)
     )
    )
    Lifetime Max=&MaskFloatBySpawnGroup(Masked Float=0.16, Passthrough Float=0.2)
    Lifetime Min=0.12
    Lifetime Mode=::NewEnumerator1
    Mesh Scale Max=&MaskVectorBySpawnGroup(Masked Vector=<1.9 1.6 1.6>, Passthrough Vector=<1.7 1.4 1.4>)
    Mesh Scale Min=&MaskVectorBySpawnGroup(Masked Vector=<1.55 1.4 1.4>, Passthrough Vector=<1.15 1 1>)
    Mesh Scale Mode=::NewEnumerator2
    UsePositionOffset=false
    Write Material Random=true
   )
   &SystemLocation(MaskSpawnGroup=false, Offset=$System.Location)
   &AddVelocity(Velocity=&MaskVectorBySpawnGroup(
     Masked Vector=&RandomRangeVector(Minimum=<-50 -35 50>, Maximum=<-80 35 70>)
     Passthrough Vector=&RandomRangeVector(Minimum=<-30 -35 15>, Maximum=<-10 35 40>)
    ))
   &AddVelocity(Velocity=&RandomRangeVector(Minimum=<0 -5 0>, Maximum=<0 -25 0>)) {module=AddVelocity001}
  ]
  particleUpdateScript [
   &ParticleState()
   &Drag(Drag=4, Use Linear Drag=true, Use Rotational Drag=false)
   &InheritVelocity(Inherited Velocity Amount Scale=<-0.1 -0.1 0>, Speed Limit=true, UseSpeedThreshold=false)
   &SolveForcesAndVelocity(Clamp Velocity=false, Limit Acceleration=false)
   &ScaleColor(
    Scale Alpha=&FloatFromCurve(FloatCurve=DataInterface() {propertyValues {Curve {keys [
         {interpMode=RCIM_Cubic, leaveTangentWeight=0.3333, tangentMode=RCTM_User, time=0, value=1}
         {arriveTangent=-2.9709, arriveTangentWeight=0.3333, interpMode=RCIM_Cubic, leaveTangent=-2.9709, tangentMode=RCTM_User, time=1, value=0}
        ]}}})
    ScaleA=true
    ScaleRGB=true
   )
   &ScaleSpriteSize(
    Non-Uniform Curve Scale=&MakeVector2D(
     X=&RemapRange(
      Input Value To Remap=&MakeFloatFromVector(Vector=$User.LeftVFast, Channel=::NewEnumerator0)
      Input Min=0
      Input Max=1
      Output Min=1
      Output Max=2.5
     )
     Y=&RemapRange(
      Input Value To Remap=&MakeFloatFromVector(Vector=$User.LeftVFast, Channel=::NewEnumerator1)
      Input Min=0
      Input Max=1
      Output Min=1
      Output Max=2.5
     )
    )
    Non-Uniform Curve Sprite Scale=DataInterface() {propertyValues {
      XCurve {keys [<at 0 0 RCIM_Cubic>, {arriveTangent=1, interpMode=RCIM_Cubic, leaveTangent=1, time=0.4977, value=0.7381}, <at 1 1 RCIM_Cubic>]}
      YCurve {keys [
        <at 0 0.2 RCIM_Cubic>
        <at 0.4977 1 RCIM_Cubic>
        {arriveTangent=-4.5774, interpMode=RCIM_Cubic, leaveTangent=-4.5774, tangentMode=RCTM_User, time=1, value=0.2}
       ]}
     }}
    Scale Sprite Size Mode=::NewEnumerator3
   )
   &DynamicMaterialParameters(
    Index 0 Param 1=&FloatFromCurve(FloatCurve=DataInterface() {propertyValues {Curve {keys [<at 0.1 0.1 RCIM_Cubic>,<at 0.8 1>]}}})
    Index 0 Param 2=2
    Index 0 Param 3=&RandomRangeFloat(Minimum=1, Maximum=1.5)
    Index 0 Param 4=&FloatFromCurve(FloatCurve=DataInterface() {propertyValues {Curve {keys [<at 0 1>,<at 1 3>]}}})
    Module.Param0WriteEnabled=true
    Module.Param1WriteEnabled=true
    Module.Param2WriteEnabled=true
    Module.Param3WriteEnabled=true
   )
  ]
  rendererClasses [&NiagaraMeshRendererProperties]
  renderers [{rendererClass=&NiagaraMeshRendererProperties, rendererIndex=0}]
 }
 SpeedLines_Ball {
  emitterUpdateScript [
   &EmitterState()
   &SpawnRate(
    SpawnRate=&RemapRange(Input Value To Remap=$User.Fast, Input Min=0.2, Input Max=3, Output Min=0, Output Max=80)
    Use Spawn Probability=false
   )
  ]
  particleSpawnScript [
   &InitializeParticle(
    Color=&MakeLinearColorFromVectorAndFloat(
     "Vector (RGB)"=&MakeVectorFromLinearColorRGB(Color=<rgba 0.2292 0.1989 0.1838 1>)
     "Float (Alpha)"=&RandomRangeFloat(Minimum=0.2, Maximum=0.4)
    )
    Lifetime Max=0.4
    Lifetime Min=0.3
    Lifetime Mode=::NewEnumerator1
    Sprite Size Max=<6 1.25>
    Sprite Size Min=<4 0.8>
    Sprite Size Mode=::NewEnumerator2
    UsePositionOffset=false
    Write Material Random=true
   )
   &ShapeLocation(
    Cylinder Height=150
    Cylinder Height Midpoint=0
    Cylinder Radius=&RandomRangeFloat(Minimum=45, Maximum=80)
    MaskSpawnGroup=false
    Offset=&RandomRangeVector(Minimum=<150 0 0>, Maximum=<100 0 0>)
    Offset Mode=::NewEnumerator0
    Override Seed=false
    Rotation Mode=::NewEnumerator1
    Shape Primitive=::NewEnumerator1
    Surface Only=true
    Use Endcaps In Surface Only Mode=false
    UseOwnerScale=true
   )
   &CameraOffset(Camera Offset Amount=10)
   &AddVelocity(Velocity=&RandomRangeVector(Minimum=<-50 0 0>, Maximum=<-100 0 0>))
  ]
  particleUpdateScript [
   &InheritVelocity(Inherited Velocity Amount Scale=<-0.9 -0.9 -0.9>, Speed Limit=false, UseSpeedThreshold=false)
   &SolveForcesAndVelocity(Clamp Velocity=false, Limit Acceleration=false)
   &ParticleState()
   &ScaleColor(
    Scale Alpha=&FloatFromCurve(FloatCurve=DataInterface() {propertyValues {Curve {keys [
         {interpMode=RCIM_Cubic, tangentMode=RCTM_User, time=0, value=0}
         {arriveTangent=2.3119, interpMode=RCIM_Cubic, leaveTangent=2.3119, tangentMode=RCTM_User, time=0.4449, value=0.904}
         {tangentMode=RCTM_User, time=1, value=0}
        ]}}})
    ScaleA=true
    ScaleRGB=true
   )
   &ScaleSpriteSize(
    Non-Uniform Curve Sprite Scale=DataInterface() {propertyValues {
      XCurve {keys [<at 0 0.5 RCIM_Cubic>,<at 0.5 1 RCIM_Cubic>,<at 1 0.5 RCIM_Cubic>]}
      YCurve {keys [<at 0 0 RCIM_Cubic>,<at 0.5 1 RCIM_Cubic>,<at 1 0 RCIM_Cubic>]}
     }}
    Scale Sprite Size Mode=::NewEnumerator3
   )
   &ScaleSpriteSizeBySpeed(Max Scale Factor=<1 100>)
  ]
  rendererClasses [&NiagaraSpriteRendererProperties]
  renderers [{rendererClass=&NiagaraSpriteRendererProperties}]
 }
 SpeedLines_Camera {
  emitterUpdateScript [
   &EmitterState()
   &SpawnRate(
    SpawnRate=&RemapRange(Input Value To Remap=$User.Fast, Input Min=0, Input Max=3, Output Min=0, Output Max=300)
    Use Spawn Probability=false
   )
  ]
  particleSpawnScript [
   &InitializeParticle(
    Color=&MakeLinearColorFromVectorAndFloat(
     "Vector (RGB)"=&MakeVectorFromLinearColorRGB(Color=<rgba 0.6875 0.6851 0.6839 1>)
     "Float (Alpha)"=&RandomRangeFloat(Minimum=0.1, Maximum=0.2)
    )
    Lifetime Max=0.35
    Lifetime Min=0.25
    Lifetime Mode=::NewEnumerator1
    Sprite Size Max=<5.5 4>
    Sprite Size Min=<4 2>
    Sprite Size Mode=::NewEnumerator2
    UsePositionOffset=false
    Write Material Random=true
   )
   &ShapeLocation(
    Cylinder Height Midpoint=0
    Cylinder Radius=500
    MaskSpawnGroup=false
    Offset=<700 0 0>
    Offset Mode=::NewEnumerator0
    Override Seed=false
    Rotation Mode=::NewEnumerator1
    Shape Primitive=::NewEnumerator1
    Surface Only=true
    Use Endcaps In Surface Only Mode=false
    UseOwnerScale=true
   )
   &CameraOffset(Camera Offset Amount=-100)
   &AddVelocity(Velocity=&RandomRangeVector(Minimum=<-50 0 0>, Maximum=<-100 0 0>))
   &AddVelocity(Velocity=&Multiply_VectorByFloat(Vector=$Particles.ShapeLocation.ShapeNormal, Float=500)) {module=AddVelocity001}
  ]
  particleUpdateScript [
   &InheritVelocity(Inherited Velocity Amount Scale=<-0.7 -0.7 -0.7>, Speed Limit=false, UseSpeedThreshold=false)
   &SolveForcesAndVelocity(Clamp Velocity=false, Limit Acceleration=false)
   &ParticleState()
   &ScaleColor(
    Scale Alpha=&FloatFromCurve(FloatCurve=DataInterface() {propertyValues {Curve {keys [
         {interpMode=RCIM_Cubic, tangentMode=RCTM_User, time=0, value=0}
         {arriveTangent=2.3119, interpMode=RCIM_Cubic, leaveTangent=2.3119, tangentMode=RCTM_User, time=0.4449, value=0.904}
         {tangentMode=RCTM_User, time=1, value=0}
        ]}}})
    ScaleA=true
    ScaleRGB=true
   )
   &ScaleSpriteSize(
    Non-Uniform Curve Sprite Scale=DataInterface() {propertyValues {
      XCurve {keys [<at 0 0.5 RCIM_Cubic>,<at 0.5 1 RCIM_Cubic>,<at 1 0.5 RCIM_Cubic>]}
      YCurve {keys [<at 0 0 RCIM_Cubic>,<at 0.5 1 RCIM_Cubic>,<at 1 0 RCIM_Cubic>]}
     }}
    Scale Sprite Size Mode=::NewEnumerator3
   )
   &ScaleSpriteSizeBySpeed(Max Scale Factor=<1 100>)
  ]
  rendererClasses [&NiagaraSpriteRendererProperties]
  renderers [{rendererClass=&NiagaraSpriteRendererProperties}]
 }
 Wheel_Smoke {
  emitterUpdateScript [
   &EmitterState()
   &SpawnRate(
    SpawnRate=&MakeCustomFloatFromBool(
     Bool=$User.SpawnDust
     True Float=&Multiply_Float(
      A=&RemapRange(Input Value To Remap=$User.Dust, Input Min=0, Input Max=1, Output Min=0, Output Max=70)
      B=&RemapRange(Input Value To Remap=$User.DustFast, Input Min=0, Input Max=1, Output Min=1, Output Max=2.2)
     )
     False Float=0
    )
    Use Spawn Probability=false
   )
  ]
  particleSpawnScript [
   &InitializeParticle(
    Color=&MakeLinearColorFromVectorAndFloat(
     "Vector (RGB)"=&MakeVectorFromLinearColorRGB(Color=<rgba 0.4427 0.4152 0.3989 1>)
     "Float (Alpha)"=&RandomRangeFloat(Minimum=0.05, Maximum=0.1)
    )
    Lifetime Max=0.36
    Lifetime Min=0.25
    Lifetime Mode=::NewEnumerator1
    Sprite Rotation Mode=::NewEnumerator2
    Sprite Size Mode=::NewEnumerator1
    Uniform Sprite Size Max=80
    Uniform Sprite Size Min=55
    UsePositionOffset=false
    Write Material Random=true
   )
   &AddVelocity(Velocity=&RandomRangeVector(Minimum=<-200 -80 150>, Maximum=<-300 -180 70>))
   &SystemLocation(MaskSpawnGroup=false, Offset=$System.Location)
   &CameraOffset(Camera Offset Amount=18)
  ]
  particleUpdateScript [
   &ParticleState()
   &Drag(Drag=10, Use Linear Drag=true, Use Rotational Drag=false)
   &SolveForcesAndVelocity(Clamp Velocity=false, Limit Acceleration=false)
   &ScaleColor(
    Scale Alpha=&FloatFromCurve(
     FloatCurve=DataInterface() {propertyValues {Curve {keys [
         {arriveTangent=26.8716, interpMode=RCIM_Cubic, leaveTangent=26.8716, leaveTangentWeight=0.0167, tangentMode=RCTM_User, time=0, value=0}
         {arriveTangentWeight=0.058, interpMode=RCIM_Cubic, leaveTangentWeight=0.29, tangentMode=RCTM_User, time=0.05, value=1}
         {arriveTangentWeight=0.121, interpMode=RCIM_Cubic, leaveTangentWeight=0.3388, tangentMode=RCTM_User, time=0.3, value=1}
         {arriveTangentWeight=0.2333, interpMode=RCIM_Cubic, tangentMode=RCTM_User, time=1, value=0}
        ]}}}
     Scale Curve=50
    )
    Scale RGB=&VectorFromFloat(Value=&FloatFromCurve(FloatCurve=DataInterface() {propertyValues {Curve {keys [{arriveTangent=1.8145, interpMode=RCIM_Cubic, leaveTangent=4.5795, tangentMode=RCTM_Break, time=0, value=0.4}, <at 0.3387 0.9565>, <at 1 1>]}}}))
    ScaleA=true
    ScaleRGB=true
   )
   &ScaleSpriteSize(Scale Factor=&Vector2DFromFloat(Value=&FloatFromCurve(FloatCurve=DataInterface() {propertyValues {Curve {keys [{arriveTangent=0.9615, interpMode=RCIM_Cubic, leaveTangent=0.9615, tangentMode=RCTM_User, time=0, value=0.45}, <at 1 1 RCIM_Cubic>]}}})))
   &ScaleSpriteSizeBySpeed(Max Scale Factor=<1 1.5>)
   &DynamicMaterialParameters(
    Index 0 Param 1=&Add_Float(
     A=$Particles.RandomFloat
     B=&FloatFromCurve(FloatCurve=DataInterface() {propertyValues {Curve {keys [
          {arriveTangent=2.7533, interpMode=RCIM_Cubic, leaveTangent=2.847, tangentMode=RCTM_Break, time=0, value=0}
          {arriveTangent=1.3581, interpMode=RCIM_Cubic, leaveTangent=1.3581, tangentMode=RCTM_User, time=0.2558, value=0.5125}
          {arriveTangent=0.0529, interpMode=RCIM_Cubic, leaveTangent=0.2021, tangentMode=RCTM_User, time=1, value=1}
         ]}}})
    )
    Index 0 Param 4=&FloatFromCurve(FloatCurve=DataInterface() {propertyValues {Curve {keys [{arriveTangent=-1.4959, interpMode=RCIM_Cubic, leaveTangent=-1.4959, tangentMode=RCTM_User, time=0, value=1}, <at 0.8 0 RCIM_Cubic>]}}})
    Index 2 Param 1=&FloatFromCurve(FloatCurve=DataInterface() {propertyValues {Curve {keys [
         {arriveTangent=1.629, interpMode=RCIM_Cubic, leaveTangent=1.629, tangentMode=RCTM_User, time=0, value=0}
         {interpMode=RCIM_Cubic, leaveTangent=-0.0719, time=1, value=1}
        ]}}})
    Index 2 Param 2=&FloatFromCurve(
     FloatCurve=DataInterface() {propertyValues {Curve {keys [
         {arriveTangent=1.9056, time=0, value=0.2}
         {arriveTangent=0.0259, interpMode=RCIM_Cubic, leaveTangent=0.0259, tangentMode=RCTM_User, time=1, value=1.4}
        ]}}}
     Scale Curve=&RandomRangeFloat(Minimum=0.5, Maximum=1)
    )
    Index 2 Param 3=&FloatFromCurve(FloatCurve=DataInterface() {propertyValues {Curve {keys [
         {interpMode=RCIM_Cubic, leaveTangent=1.6063, tangentMode=RCTM_Break, time=0, value=1}
         {interpMode=RCIM_Cubic, leaveTangent=11.1233, tangentMode=RCTM_Break, time=1, value=2}
        ]}}})
    Index 2 Param 4=&FloatFromCurve(
     FloatCurve=DataInterface() {propertyValues {Curve {keys [
         {arriveTangent=-0.6867, interpMode=RCIM_Cubic, leaveTangent=-0.6867, tangentMode=RCTM_User, time=0, value=1}
         {interpMode=RCIM_Cubic, leaveTangent=-1.2956, tangentMode=RCTM_Break, time=1, value=0.75}
        ]}}}
     Scale Curve=1.6
    )
    Module.Param0WriteEnabled=true
    Module.Param0WriteEnabled1=true
    Module.Param0WriteEnabled2=true
    Module.Param1WriteEnabled=false
    Module.Param1WriteEnabled1=true
    Module.Param1WriteEnabled2=true
    Module.Param2WriteEnabled=true
    Module.Param2WriteEnabled1=true
    Module.Param2WriteEnabled2=true
    Module.Param3WriteEnabled=true
    Module.Param3WriteEnabled1=true
    Module.Param3WriteEnabled2=true
    Write Parameter Index 1=true
    Write Parameter Index 2=true
   )
  ]
  rendererClasses [&NiagaraSpriteRendererProperties]
  renderers [
   {rendererClass=&NiagaraSpriteRendererProperties}
   {rendererClass=&NiagaraSpriteRendererProperties, rendererIndex=1}
  ]
 }
}
summary {
 systemName=NS_Octopus_Vehicle
 emitters [
  {emitterName=Coil_Meshes, rendererClasses [&NiagaraMeshRendererProperties, &NiagaraSpriteRendererProperties]}
  {emitterName=Electric_Arcs, rendererClasses [&NiagaraMeshRendererProperties, &NiagaraLightRendererProperties]}
  {emitterName=DamagePoint, simTarget=GPUComputeSim, rendererClasses [&NiagaraSpriteRendererProperties]}
  {bEnabled=false, emitterName=DamageSparks, simTarget=GPUComputeSim, rendererClasses [&NiagaraSpriteRendererProperties]}
  {emitterName=Damaged_Smoke, rendererClasses [&NiagaraSpriteRendererProperties]}
  {emitterName=Wheel_Smoke, rendererClasses [&NiagaraSpriteRendererProperties]}
  {emitterName=Directional_Smoke, rendererClasses [&NiagaraSpriteRendererProperties]}
  {emitterName=Skid_Mesh, rendererClasses [&NiagaraMeshRendererProperties]}
  {emitterName=Skid_Chunks, simTarget=GPUComputeSim, rendererClasses [&NiagaraSpriteRendererProperties]}
  {emitterName=SpeedLines_Camera, rendererClasses [&NiagaraSpriteRendererProperties]}
  {emitterName=SpeedLines_Ball, rendererClasses [&NiagaraSpriteRendererProperties]}
  {bEnabled=false, emitterName=Empty, rendererClasses []}
  {bEnabled=false, emitterName=Empty001, rendererClasses []}
 ]
 userVariables [
  User.BadlyDamaged {type=NiagaraFloat, defaultValue=0, description=""}
  User.Boost {type=NiagaraFloat, defaultValue=0, description=""}
  User.BoostAlpha {type=NiagaraFloat, defaultValue=0, description=""}
  User.Damage {type=NiagaraFloat, defaultValue=0, description=""}
  User.Dust {type=NiagaraFloat, defaultValue=0, description=""}
  User.DustFast {type=NiagaraFloat, defaultValue=0, description=""}
  User.Fast {type=NiagaraFloat, defaultValue=0, description=""}
  User.LeftVFast {type=Vector3f, description="", defaultValue {struct=Vector3f, value=<0 0 0>}}
  User.SpawnDust {type=NiagaraBool, defaultValue=true}
  User.Speed {type=NiagaraFloat, defaultValue=0, description=""}
 ]
}
